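; Player-controller component that directs playlist travel UI: the initiator menu, a HUD toast and a dimmer.
; It listens on the gameplay event router for show-initiator requests and stateful travel-info changes, adding or removing Dynamic UI scenes, with the dimmer gated by a CVar.
; When Dynamic UI creates the menu or toast widgets, it feeds their view models the stored initiator options and travel info.
; Menu deactivation removes its scene.
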(event EventBeginPlay
  (Utilities|FlowControl|Sequence
    (:then_0
      (bind _asplayer_controller (Utilities|Casting|CastToPlayerController (Components|GetOwner))
        (:then
          (Variables|Default|SetOwningPlayerController _asplayer_controller)
          (bind _eventrouter (CallFunction|GetEventRouter _asplayer_controller))
          (bind (_asynctaskproxy _payload) (GameplayEventRouter|StartListeningToEvent _eventrouter "/Script/FortPlaylistTravelRuntime.FortPlaylistTravelEvent_RequestShowTravelInitiator" _asplayer_controller)
            (:then
              (Variables|Default|SetGameTravelTaskHandle _asynctaskproxy)
              (bind _eventrouter_2 (CallFunction|GetEventRouter 0))
              (bind (_asynctaskproxy_2 _payload_2) (GameplayEventRouter|StartListeningToStatefulEvent _eventrouter_2 "/Script/FortPlaylistTravelRuntime.FortPlaylistTravelEvent_TravelInfoChange" (Variables|Default|GetOwningPlayerController))
                (:then
                  (Variables|Default|SetGameTravelInfoTaskHandle _asynctaskproxy_2))
                (:OnEventReceived
                  _)
                (:OnSavedEventStateExists
                  _))
              (Variables|Default|SetCurrentTravelInfo (Utilities|Struct|BreakFortPlaylistTravelEventTravelInfoChange _payload_2))
              (switch Utilities|FlowControl|Switch|SwitchonEFortPlaylistTravelStatus (Utilities|Struct|BreakFortPlaylistTravelInfo (Utilities|Struct|BreakFortPlaylistTravelEventTravelInfoChange _payload_2))
                (:None
                  (CallFunction|HandleAddRemoveDimmer false)
                  (CallFunction|HandleAddRemoveHUDToast false))
                (:PrepTime
                  _)
                (:Launching
                  _))
              (CallFunction|HandleAddRemoveDimmer true)
              (CallFunction|HandleAddRemoveHUDToast true))
            (:OnEventReceived
              (Variables|Default|SetInitiator%20Options (Utilities|Struct|BreakFortPlaylistTravelEventRequestShowTravelInitiator _payload))
              (CallFunction|OpenPlaylistTravelMenu))))
        (:CastFailed
          (Utilities|Debugging|RaiseScriptError "Couldn't get player controller from owner of component"))))
    (:then_1
      ;[1 1 1] Listen for when the Dynamic UI Manager adds the playlist travel menu widget
      (FortPlaylistTravelUI|DynamicUI|AssignOnFortWidgetsCreatedinDynamicUI (delegate Custom|OnFortWidgetsCreatedInDynamicUI_Event)))))

(event EventEndPlay
  (params (EndPlayReason :type EEndPlayReason%20Enum :default "Destroyed"))
  (GameplayEventRouter|StopListeningtoEvent (Variables|Default|GetGameTravelTaskHandle))
  (GameplayEventRouter|StopListeningtoStatefulEvent (Variables|Default|GetGameTravelInfoTaskHandle)))

(event Custom|OnFortWidgetsCreatedInDynamicUI_Event
  (params (Widgets :type Array%20of%20Widget%20Object%20References))
  ;[1 1 1] Listen for when the Dynamic UI Manager adds the playlist travel menu widget
  (Utilities|FlowControl|Sequence
    (:then_0
      (bind _array_element (Utilities|Array|ForEachLoopwithBreak Widgets)
        (:LoopBody
          (bind _aswbp_fort_playlist_travel_menu (Utilities|Casting|CastToWBP_FortPlaylistTravel_Menu _array_element)
            (:then
              (bind _output_get (Variables|Default|SetPlaylistTravelMenu _aswbp_fort_playlist_travel_menu))
              ;[1 1 1] We setup the view model with the initiator options passed in the event that we stored.
              (Utilities|IsValid (Variables|/FortPlaylistTravelUI/Menu/WBP_FortPlaylistTravel_Menu.WBP_FortPlaylistTravel_Menu_C|GetBP_FortPlaylistTravelViewModel _output_get)
                (:Is%20Valid
                  (Class|FortPlaylistTravelViewModel|SetupViewModel (Variables|/FortPlaylistTravelUI/Menu/WBP_FortPlaylistTravel_Menu.WBP_FortPlaylistTravel_Menu_C|GetBP_FortPlaylistTravelViewModel _output_get) (Variables|Default|GetInitiator%20Options) (Variables|Default|GetCurrentTravelInfo))
                  ;[1 1 1] When the menu gets deactivated, we also want to remove the menu scene
                  (Events|BindEventtoOnWidgetDeactivated _output_get (delegate Custom|MenuDeactivated)))
                (:Is%20Not%20Valid
                  _)))
            (:CastFailed
              _))
          (bind _aswbp_fort_playlist_travel_hudtoast (Utilities|Casting|CastToWBP_FortPlaylistTravel_HUDToast _array_element))
          (Utilities|IsValid (Variables|/FortPlaylistTravelUI/HUD/WBP_FortPlaylistTravel_HUDToast.WBP_FortPlaylistTravel_HUDToast_C|GetBP_FortPlaylistTravelViewModel _aswbp_fort_playlist_travel_hudtoast)
            (:Is%20Valid
              (Class|FortPlaylistTravelViewModel|SetupViewModel (Variables|/FortPlaylistTravelUI/HUD/WBP_FortPlaylistTravel_HUDToast.WBP_FortPlaylistTravel_HUDToast_C|GetBP_FortPlaylistTravelViewModel _aswbp_fort_playlist_travel_hudtoast) (Variables|Default|GetInitiator%20Options) (Variables|Default|GetCurrentTravelInfo)))))))
    (:then_1
      (Utilities|IsValid (Variables|FortPlaylistTravelUIComponent|GetCurrentUIDirector)
        (:Is%20Not%20Valid
          (bind _returnvalue (Actor|GetActorOfClass "/Script/DynamicUI.DynamicUIDirectorBase"))
          (Variables|FortPlaylistTravelUIComponent|SetCurrentUIDirector _returnvalue))))))

(event Custom|MenuDeactivated
  ;[1 1 1] When the menu gets deactivated, we also want to remove the menu scene
  (FortPlaylistTravelUI|DynamicUI|RemoveDynamicUIScene (Variables|Default|GetOwningPlayerController) (Variables|Default|GetWidgetScene)))

(fn GetEventRouter
  (params (ContextActor :type Actor%20Object%20Reference))
  (Utilities|IsValid (Variables|Default|GetGameplayEventRouter)
    (:Is%20Valid
      _)
    (:Is%20Not%20Valid
      (Variables|Default|SetGameplayEventRouter (GameplayEventRouter|GetGlobalEventRouter))))
  (return (Variables|Default|GetGameplayEventRouter)))

(fn HandleAddRemoveDimmer
  (params (Add :type Boolean))
  (bind _owningplayercontroller (Variables|Default|GetOwningPlayerController))
  (bind _owningplayercontroller_2 (Variables|Default|GetOwningPlayerController))
  (bind _returnvalue (Development|GetConsoleVariableBoolValue (Variables|Default|GetShowDimmerCVar)))
  (if _returnvalue
    (if Add
      (Utilities|IsValid _owningplayercontroller
        (:Is%20Valid
          (FortPlaylistTravelUI|DynamicUI|AddDynamicUIScene _owningplayercontroller (Variables|Default|GetGameTravelDimmer)))
        (:Is%20Not%20Valid
          (Utilities|Debugging|RaiseScriptError "Couldn't had Travel dimmer because the owning player controller is invalid")))
      (else
        (Utilities|IsValid (Variables|Default|GetGameTravelDimmer)
          (:Is%20Valid
            (Utilities|IsValid _owningplayercontroller_2
              (:Is%20Valid
                (FortPlaylistTravelUI|DynamicUI|RemoveDynamicUIScene _owningplayercontroller_2 (Variables|Default|GetGameTravelDimmer)))
              (:Is%20Not%20Valid
                (Utilities|Debugging|RaiseScriptError "Could not removeTravel dimmer because the owning player controller is invalid"))))
          (:Is%20Not%20Valid
            (Utilities|Debugging|RaiseScriptError "Could not remove Travel dimmer because the reference to it is invalid")))))
    (else
      (Development|PrintString "Not showing dimmer. CVar off."))))

(fn HandleAddRemoveHUDToast
  (params (Add :type Boolean))
  (bind _owningplayercontroller (Variables|Default|GetOwningPlayerController))
  (bind _owningplayercontroller_2 (Variables|Default|GetOwningPlayerController))
  (if Add
    (Utilities|IsValid _owningplayercontroller
      (:Is%20Valid
        (FortPlaylistTravelUI|DynamicUI|AddDynamicUIScene _owningplayercontroller (Variables|Default|GetGameTravelHUDToast)))
      (:Is%20Not%20Valid
        (Utilities|Debugging|RaiseScriptError "Couldn't had Travel HUD Toast because the owning player controller is invalid")))
    (else
      (Utilities|IsValid (Variables|Default|GetGameTravelHUDToast)
        (:Is%20Valid
          (Utilities|IsValid _owningplayercontroller_2
            (:Is%20Valid
              (FortPlaylistTravelUI|DynamicUI|RemoveDynamicUIScene _owningplayercontroller_2 (Variables|Default|GetGameTravelHUDToast)))
            (:Is%20Not%20Valid
              (Utilities|Debugging|RaiseScriptError "Could not remove travel HUD toast because the owning player controller is invalid"))))
        (:Is%20Not%20Valid
          (Utilities|Debugging|RaiseScriptError "Could not remove travel HUD toast because the reference to it is invalid"))))))

(fn OpenPlaylistTravelMenu
  (bind _owningplayercontroller (Variables|Default|GetOwningPlayerController))
  (Utilities|IsValid _owningplayercontroller
    (:Is%20Valid
      (FortPlaylistTravelUI|DynamicUI|AddDynamicUIScene _owningplayercontroller (Variables|Default|GetWidgetScene)))
    (:Is%20Not%20Valid
      (Utilities|Debugging|RaiseScriptError "Could not open playlist travel menu scene because the player controller is not valid"))))

(fn StopListeningToTravelInfoEvents
  (bind _gametravelinfotaskhandle (Variables|Default|GetGameTravelInfoTaskHandle))
  ;[1 1 1] Use this function if you don't want to use the HUD popup when you begin traveling
  (Utilities|IsValid _gametravelinfotaskhandle
    (:Is%20Valid
      (GameplayEventRouter|StopListeningtoStatefulEvent _gametravelinfotaskhandle))))
